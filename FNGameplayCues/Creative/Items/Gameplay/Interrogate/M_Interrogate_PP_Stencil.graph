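# Post-process material for the Interrogate effect that highlights stencil-tagged players in EnemyColor (red by default) and chests in yellow where occluded.
# MPC origin, radius and timestamps drive expanding, fading radial masks that gate the highlight.
# Custom-depth comparison isolates hidden pixels, adding four-tap outlines and a pulsing fill over scene color; mobile passes the scene through.
# The world-aligned grid and depth-edge scan-wave branch is left unconnected to the final output.
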Material /FNGameplayCues/Creative/Items/Gameplay/Interrogate/M_Interrogate_PP_Stencil
graph {
 properties {
  blendMode=BLEND_Translucent
  blendableLocation=BL_SceneColorBeforeDOF
  blendablePriority=-5
  materialDomain=MD_PostProcess
  pixelDepthOffsetMode=PDOM_Legacy
  shadingModel=MSM_Unlit
  usedShadingModels=Unlit
 }
 nodes [
  n1 = ComponentMask(None=CollectionParameter() {collection=&/MPC_Interrogate_StencilOnly, parameterName=OriginPositionWS}) {b=true, g=true, r=true}
  n2 = MaterialFunctionCall(Float3=WorldPosition().XYZ) {materialFunction=&BreakOutFloat3Components}
  n3 = ScalarParameter() {defaultValue=6000, parameterName=HeightMaskRadius}
  n4 = ScalarParameter() {defaultValue=0.25, parameterName=HeightMaskHardness}
  n5 = ScalarParameter() {defaultValue=8, parameterName=TipScale}
  n6 = Constant3Vector() {constant=<rgba 1 1 0 0>}
  n7 = Multiply(A=@n1, B=@n6)
  n8 = Multiply(A=WorldPosition().XYZ, B=@n6)
  n9 = MaterialFunctionCall(
   TimestampA=CollectionParameter() {collection=&MPC_Interrogate_StencilOnly, parameterName=ActivatedTimestamp}
   TransitionLength_In=CollectionParameter() {collection=&MPC_Interrogate_StencilOnly, parameterName=Duration}
   Transition In Only?=StaticBool() {value=true}
  ) {materialFunction=&MF_Timestamp_InOut_01}
  n10 = Multiply(
   A=CollectionParameter() {collection=&MPC_Interrogate_StencilOnly, parameterName=Radius}
   B=@n9.Linear
  )
  n11 = Constant()
  n12 = Power(Base=OneMinus(None=SphereMask(A=@n7, B=@n8, Radius=@n10, Hardness=@n11)), Exp=ScalarParameter() {defaultValue=1.5, parameterName=Exponent})
  n13 = SphereMask(
   A=@n12
   B=ScalarParameter() {defaultValue=0.94, parameterName=TipCenter}
   Radius=ScalarParameter() {defaultValue=0.05, parameterName=TipRadius}
   Hardness=ScalarParameter() {defaultValue=0.05, parameterName=TipHardness}
  )
  n14 = MaterialFunctionCall(
   TextureObject=TextureObject() {samplerType=SAMPLERTYPE_LinearColor, texture=/Game/Gadgets/Assets/Teleporter/Textures/T_Teleporter_GridSquares_01}
   TextureSize=ScalarParameter() {defaultValue=256, parameterName=WorldTextureScale}
   World Space Normal=ShadingPathSwitch(Default=VertexNormalWS(), Deferred=SceneTexture() {sceneTextureId=PPI_WorldNormal}.Color)
  ) {materialFunction=&WorldAlignedTexture}
  n15 = ComponentMask(None=@n14.XYZ Texture) {r=true}
  n16 = ConstantBiasScale(None=Sine(None=Add(A=@n9.Linear, B=@n15))) {bias=2}
  n17 = SphereMask(
   A=@n12
   B=ScalarParameter() {defaultValue=0.5, parameterName=RingCenter}
   Radius=ScalarParameter() {defaultValue=0.5, parameterName=RingRadius}
   Hardness=ScalarParameter() {parameterName=RingHardness}
  )
  n18 = ScreenPosition()
  n19 = Multiply(B=SceneTexelSize()) {constA=3}
  n20 = Add(A=@n18.ViewportUV, B=Multiply(A=@n19, B=Constant2Vector() {g=-1, r=-1}))
  n21 = SceneTexture() {sceneTextureId=PPI_SceneDepth}
  n22 = Add(A=@n18.ViewportUV, B=Multiply(A=@n19, B=Constant2Vector() {g=-1, r=1}))
  n23 = Add(
   A=Subtract(A=SceneTexture(UVs=@n20) {sceneTextureId=PPI_SceneDepth}.Color, B=@n21.Color)
   B=Subtract(A=SceneTexture(UVs=@n22) {sceneTextureId=PPI_SceneDepth}.Color, B=@n21.Color)
  )
  n24 = Multiply(A=@n19, B=Constant2Vector() {g=1, r=1})
  n25 = Subtract(A=SceneTexture(UVs=Add(A=@n18.ViewportUV, B=@n24)) {sceneTextureId=PPI_SceneDepth}.Color, B=@n21.Color)
  n26 = Constant2Vector() {g=1, r=-1}
  n27 = SceneTexture(UVs=Add(A=@n18.ViewportUV, B=Multiply(A=@n19, B=@n26))) {sceneTextureId=PPI_SceneDepth}
  n28 = Abs(None=Add(A=Add(A=@n23, B=@n25), B=Subtract(A=@n27.Color, B=@n21.Color)))
  n29 = Saturate(None=Multiply(A=ComponentMask(None=@n28) {r=true}, B=ScalarParameter() {defaultValue=0.01, parameterName=EdgeDetectionScaleA}))
  n30 = ShadingPathSwitch(Default=Constant() {r=1}, Deferred=Multiply(A=Round(None=@n29), B=ScalarParameter() {defaultValue=10, parameterName=EdgeDetectionScaleB}))
  n31 = ComponentMask(None=SceneTexture() {sceneTextureId=PPI_PostProcessInput0}.Color) {b=true, g=true, r=true}
  n32 = Multiply(
   A=Add(A=@n17, B=Multiply(A=@n17, B=@n30))
   B=Desaturation(None=@n31, Fraction=ScalarParameter() {defaultValue=0.5, parameterName=DesaturationFraction})
  )
  n33 = OneMinus(None=ComponentMask(None=@n14.XYZ Texture) {g=true})
  n34 = Add(
   A=Multiply(A=SphereMask(A=@n1, B=@n2.B, Radius=@n3, Hardness=@n4), B=Multiply(A=@n5, B=@n13))
   B=Add(A=Multiply(A=@n16, B=@n32), B=Multiply(A=@n17, B=@n33))
  )
  n35 = VectorParameter() {parameterName=Color, defaultValue=<rgba 0.0057 0.1038 0.3 0>}
  n36 = Saturate(None=Multiply(A=OneMinus(None=@n9.Linear)) {constB=4})
  Add(A=Multiply(A=Multiply(A=@n34, B=@n35.RGB), B=@n36), B=@n31)
  n37 = Add(A=SceneTexture() {sceneTextureId=PPI_CustomDepth}.Color, B=ScalarParameter() {defaultValue=-1, parameterName=Custom Depth Offset Amount})
  n38 = SceneTexture() {sceneTextureId=PPI_SceneDepth}
  ComponentMask(None=Subtract(A=@n37, B=@n38.Color)) {r=true}
  n39 = Time()
  n40 = ScalarParameter() {defaultValue=0.6666, parameterName=TimeScale}
  Multiply(A=ScalarParameter() {defaultValue=9000, parameterName=SourceRadius}, B=Fmod(A=Multiply(A=@n39, B=@n40), B=Constant() {r=1}))
  n41 = Multiply(A=ComponentMask(None=CollectionParameter() {collection=&MPC_Interrogate_StencilOnly, parameterName=OriginPositionWS}) {b=true, g=true, r=true})
  n42 = Multiply(A=WorldPosition().XYZ, B=Constant3Vector() {constant=<rgba 1 1 0 0>})
  n43 = MaterialFunctionCall(
   TimestampA=CollectionParameter() {collection=&MPC_Interrogate_StencilOnly, parameterName=ActivatedTimestamp}
   TransitionLength_In=CollectionParameter() {collection=&MPC_Interrogate_StencilOnly, parameterName=StencilDuration}
   Transition In Only?=StaticBool() {value=true}
  ) {materialFunction=&MF_Timestamp_InOut_01}
  n44 = Multiply(
   A=CollectionParameter() {collection=&MPC_Interrogate_StencilOnly, parameterName=Radius}
   B=@n43.Linear
  )
  n45 = SphereMask(A=@n41, B=@n42, Radius=@n44, Hardness=Constant() {r=0.5})
  n46 = Constant()
  Multiply(A=@n45, B=OneMinus(None=SphereMask(A=@n41, B=@n42, Radius=@n44, Hardness=@n46)))
  Power(Base=@n13, Exp=ScalarParameter() {defaultValue=2, parameterName=ExponentTip})
  n47 = Subtract(
   A=SceneTexture() {sceneTextureId=PPI_SceneDepth}.Color
   B=Add(A=SceneTexture() {sceneTextureId=PPI_CustomDepth}.Color, B=ScalarParameter() {defaultValue=-1, parameterName=Custom Depth Offset Amount})
  )
  n48 = OneMinus(None=Ceil(None=Clamp(None=@n47)) {desc=z tested objects black z tested - everything else white}) {desc=occlusion mask}
  Reroute(None=@n48)
  Reroute(None=@n48)
  ShadingPathSwitch()
  n49 = MaterialFunctionCall!(ID=CollectionParameter() {collection=/Game/Packages/Fortress_SharedMaterials/Stencil_ID_Definitions/Stencil_ID_Definitions, parameterName=InterrogationPlayer}) {materialFunction=/Game/Packages/Fortress_SharedMaterials/Stencil_ID_Definitions/StencilMaskCompare}
  n50 = Multiply(A=@n17, B=@n36)
  n51 = CollectionParameter() {collection=&MPC_Interrogate_StencilOnly, parameterName=StencilDuration}
  n52 = Multiply(A=Clamp(None=@n45), B=Clamp(None=LinearInterpolate(A=@n51, Alpha=@n43.Linear) {constA=1, constB=0}))
  n53 = Subtract(A=@n38.Color, B=@n37)
  n54 = OneMinus(None=Ceil(None=Clamp(None=@n53)) {desc=z tested objects black z tested - everything else white}) {desc=occlusion mask}
  n55 = Time()
  n56 = ScalarParameter() {defaultValue=0.75, parameterName=PulseRate}
  n57 = ConstantBiasScale(None=Sine(None=Multiply(A=@n55, B=@n56)))
  n58 = ComponentMask(None=ScreenPosition().ViewportUV) {g=true, r=true}
  n59 = Subtract(A=Divide(A=ViewSize(), B=Constant() {r=320}), B=Constant() {r=2}) {constB=0}
  n60 = Multiply(A=Clamp(None=Ceil(None=@n59)) {maxDefault=2, minDefault=1}, B=SceneTexelSize()) {constA=2}
  n61 = Constant()
  n62 = AppendVector(A=ComponentMask(None=@n60) {desc=x, r=true}, B=@n61)
  n63 = Add(A=@n58, B=@n62) {desc=x+}
  n64 = AppendVector(A=@n61, B=ComponentMask(None=@n60) {desc=y, g=true}) {desc=y+}
  n65 = Add(A=@n58, B=@n64) {desc=y+}
  n66 = Constant() {r=-1}
  n67 = Add(A=@n58, B=Multiply(A=@n62, B=@n66)) {desc=x-}
  n68 = Add(A=@n58, B=Multiply(A=@n66, B=@n64)) {desc=y-}
  n69 = ComponentMask(None=MaterialFunctionCall!(
    Base=SceneTexture() {sceneTextureId=PPI_CustomDepth}.Color
    BaseSampleBias
    Scale
    Sample1=SceneTexture(UVs=@n63) {sceneTextureId=PPI_CustomDepth}.Color
    Sample2=SceneTexture(UVs=@n65) {sceneTextureId=PPI_CustomDepth}.Color
    Sample3=SceneTexture(UVs=@n67) {sceneTextureId=PPI_CustomDepth}.Color
    Sample4=SceneTexture(UVs=@n68) {sceneTextureId=PPI_CustomDepth}.Color
    Divisor=Constant() {r=-16}
    UseCenteredLines
    UseRGB
   ) {materialFunction=/Game/Packages/jl_tests/combine4Samples}.Result) {r=true}
  n70 = Multiply(A=@n54, B=Add(A=LinearInterpolate(Alpha=@n57) {constA=0.005, constB=0.02}, B=@n69))
  n71 = Multiply(A=Clamp(None=@n54), B=LinearInterpolate(Alpha=@n69) {constA=0.2})
  n72 = Clamp(None=Multiply(A=Add(A=@n50, B=@n52), B=Add(A=@n70, B=@n71)))
  n73 = LinearInterpolate(
   B=VectorParameter() {parameterName=EnemyColor, defaultValue=<rgba 1 0 0 0>}.RGB
   Alpha=ComponentMask(None=Multiply(A=@n49.Result, B=@n72)) {r=true}
  )
  n74 = Constant3Vector() {constant=<rgba 1 0.8181 0 0>}
  n75 = CollectionParameter() {collection=&Stencil_ID_Definitions, parameterName=InterrogationChest}
  n76 = ComponentMask(None=Multiply(
    A=MaterialFunctionCall(ID=@n75) {materialFunction=&StencilMaskCompare}.Result
    B=@n72
   )) {r=true}
  n77 = Clamp(None=Add(A=@n31, B=LinearInterpolate(A=@n73, B=@n74, Alpha=@n76)) {constB=0})
  ShadingPathSwitch(Default=@n77, Deferred=@n77, Forward=@n77, Mobile=@n31)
 ]
}
